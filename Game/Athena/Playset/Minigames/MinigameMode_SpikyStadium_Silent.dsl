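; MinigameMode_SpikyStadium_Silent is a minigame mode Blueprint for an Athena playset, named as a silent variant of Spiky Stadium.
; It declares event handlers for minigame warmup, start and end, plus a construction script.
; All four are empty, so the file adds no visible logic of its own to those phases.
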(event EventHandleMinigameWarmup)

(event Minigame|EventHandleMinigameStarted)

(event Minigame|EventHandleMinigameEnded)

(fn ConstructionScript)
