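; Player-created Creative minigame Blueprint that runs team assignment, match audio and end-of-round scoreboard or post-game UI.
; The authority deals shuffled players round-robin into capacity-limited teams, or refreshes voice and tallies existing team sizes.
; On cosmetic-capable instances, state changes drive stingers, sound mixes and an async-created scoreboard widget for local participants or spectators.
; It also tracks Prop Hunt hider counts with OnRep dispatchers and resolves per-item drop-or-destroy overrides, cleared when the minigame ends.
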(event Minigame|GameRules|EventSetupTeams
  (Networking|SwitchHasAuthority
    (:Authority
      (CallFunction|AssignTeams))))

(event Custom|ShowScoreboardPanel
  (params (ShowCustomScoreboard :type Boolean) (Custom%20Scoreboard%20Display%20Time :type Float%20%28double-precision%29))
  ;[0 1 0.1174] Show Scoreboard Panel
  (Variables|Default|GetScoreboardPanel
    (:then)
    (:else
      (bind _scoreboardcomponent (Variables|Minigame|GetScoreboardComponent)
        (:then
          (UI|DisableCustomScoreboard _scoreboardcomponent)
          (bind _userwidget (Class|FortAsyncActionCreateWidgetAsync|CreateWidgetAsync "/Game/Items/Traps/Blueprints/Toys/ScoreboardPanel.ScoreboardPanel_C")
            (:OnComplete
              (Utilities|IsValid _userwidget
                (:Is%20Valid
                  (if ShowCustomScoreboard
                    (UI|EnableCustomScoreboard (Variables|Minigame|GetScoreboardComponent) Custom%20Scoreboard%20Display%20Time)
                    (else
                      (Utilities|IsValid (Widget|GetOwningPlayer _userwidget)
                        (:Is%20Valid
                          (if (CallFunction|ShouldShowScoreboard (Widget|GetOwningPlayer _userwidget))
                            (bind _output_get (Variables|Default|SetScoreboardPanel _userwidget))
                            (bind _asfort_hide_show_interface (Utilities|Casting|CastToFortHideShowInterface _output_get))
                            (Class|FortHideShowInterface|DoShowUI _asfort_hide_show_interface)
                            (Variables|Default|SetIs%20Current%20A%20Round (< (Variables|FortMinigame|GetCurrentRound) (Variables|FortMinigame|GetTotalRounds)))
                            (else
                              _)))
                        (:Is%20Not%20Valid
                          _))
                      ;;[1 0.5104 0.5104] If we are not supposed to show the scoreboard, destroy it.
                      (Widget|RemovefromParent _userwidget)
                      (Variables|Default|SetScoreboardPanel 0))))))))
        (:else)))))

(event Custom|HideScoreboardPanel
  ;[1 0.488 0] Hide Scoreboard Panel
  (bind _scoreboardcomponent (Variables|Minigame|GetScoreboardComponent)
    (:then
      (UI|DisableCustomScoreboard _scoreboardcomponent)
      (bind _scoreboardpanel (Variables|Default|GetScoreboardPanel)
        (:then
          (Utilities|FlowControl|Sequence
            (:then_0
              (bind _asfort_hide_show_interface (Utilities|Casting|CastToFortHideShowInterface _scoreboardpanel))
              (Class|FortHideShowInterface|DoHideUI _asfort_hide_show_interface))
            (:then_1
              (Variables|Default|SetScoreboardPanel 0)
              (CallFunction|EndScoreboardSound))))
        (:else)))
    (:else)))

(event EventBeginPlay
  (|CallParentFunction)
  (Networking|CanExecuteCosmeticEvents
    (:True
      (Minigame|BindEventtoPlayerRemovedDelegate (EventDispatchers|CreateEvent :!event Custom|OnPlayerRemoved))
      ;[1 1 1] If player is already in the mini game at this point then we should immediately register for that delegate
      (Minigame|BindEventtoStateChangedDelegate (EventDispatchers|CreateEvent :!event Custom|StateChangedDelegate_Event)))))

(event Custom|End%20Scoreboard%20Sound
  (Utilities|FlowControl|Delay 1.25)
  (Audio|PushSoundMixModifier 0)
  (if (Variables|Default|GetIs%20Current%20A%20Round)
    (Audio|PlaySound2D "/Game/Sounds/Creative/Modes/Match/Match_Round_Change_01_Cue.Match_Round_Change_01_Cue")
    (else
      (Audio|PlaySound2D "/Game/Sounds/Creative/Modes/Match/Match_Stop_01_Cue.Match_Stop_01_Cue"))))

(event EventEndPlay
  (params (EndPlayReason :type EEndPlayReason%20Enum :default "Destroyed"))
  (|CallParentFunction EndPlayReason)
  (Networking|CanExecuteCosmeticEvents
    (:True
      (Minigame|UnbindallEventsfromPlayerAddedDelegate)
      (Minigame|UnbindallEventsfromPlayerRemovedDelegate))))

(event Custom|ClientHandleStartEndgame
  (Utilities|FlowControl|Delay 0.1)
  (Audio|PlaySound2D "/Game/Sounds/Fort_Music/Stingers/Victory_Royale_Sound_Cue.Victory_Royale_Sound_Cue"))

(event Custom|OnPlayerRemoved
  (params (UniqueNetId :type Unique%20Net%20Id%20Structure) (bIsLocalPlayer :type Boolean))
  (Networking|CanExecuteCosmeticEvents
    (:True
      (if (and bIsLocalPlayer (not (Minigame|IsPlayerinMinigamebyNetId UniqueNetId true)))
        (CallFunction|HideScoreboardPanel)))))

(event Custom|StateChangedDelegate_Event
  (params (Minigame :type Fort%20Minigame%20Object%20Reference) (MinigameState :type EFortMinigameState%20Enum :default "PreGame"))
  (if (CallFunction|IsAnyLocalPlayerinMinigame)
    (CallFunction|PlayMinigameStateAudio MinigameState)
    (switch Utilities|FlowControl|Switch|SwitchonEFortMinigameState MinigameState
      (:PreGame
        _)
      (:Warmup
        _)
      (:InProgress
        _)
      (:PostGameTimeDilation
        (CallFunction|ClientHandleStartEndgame))
      (:PostRoundEnd
        (CallFunction|ShowScoreboardPanel false))
      (:PostGameEnd
        (CallFunction|ShowPostGame))
      (:PostGameAbandon
        _)
      (:PostGameReset
        _))
    (CallFunction|HideScoreboardPanel)))

(event Custom|Play%20Minigame%20State%20Audio
  (params (Minigame%20State :type EFortMinigameState%20Enum :default "PreGame"))
  (switch Utilities|FlowControl|Switch|SwitchonEFortMinigameState Minigame%20State
    (:PreGame
      _)
    (:Setup
      _)
    (:Transitioning
      _)
    (:WaitingForCameras
      _)
    (:Warmup
      _)
    (:InProgress
      ;[1 1 1] Disable the match start sound if countdown is disabled and we aren't in a hub, or the match start sound is disabled
      (if (and (or (Minigame|IsCuratedHub) (> (Variables|/Script/FortniteGame.FortMinigameWarmupComponent|GetWarmupDuration (Variables|Minigame|GetWarmupComponent)) 0)) (Variables|/Script/FortniteGame.FortMinigameWarmupComponent|GetbGameStartSoundEnabled (Variables|Minigame|GetWarmupComponent)))
        (Audio|PlaySound2D "/Game/Sounds/Creative/Modes/Match/Match_Start_01_Cue.Match_Start_01_Cue")))
    (:PostGameTimeDilation
      (Audio|PushSoundMixModifier "/Game/Sounds/Creative/MixModifiers/Creative_MiniGameMatch_MixMod.Creative_MiniGameMatch_MixMod"))
    (:PostRoundEnd
      _)
    (:PostGameEnd
      _)
    (:PostGameAbandon
      _)
    (:PostGameReset
      _))
  (Audio|PopSoundMixModifier "/Game/Sounds/Creative/MixModifiers/Creative_MiniGameMatch_MixMod.Creative_MiniGameMatch_MixMod"))

(event Minigame|EventHandleMinigameStarted
  (bind _numberofhidersinprogress (Variables|Default|GetNumberOfHidersInProgress))
  ;[1 1 1] Setting Up values for Hide and Seek
  (|CallParentFunction)
  (if (> _numberofhidersinprogress 0)
    (Variables|Default|SetNumberOfHidersGameStart _numberofhidersinprogress)))

(event Minigame|EventHandleMinigameEnded
  (|CallParentFunction)
  (Utilities|Map|Clear (Variables|Default|GetItemDropOverride)))

(fn AddPropHuntHider
  (Networking|CanExecuteCosmeticEvents
    (:False
      (Variables|Default|SetNumberOfHidersInProgress (+ (Variables|Default|GetNumberOfHidersInProgress) 1)))))

(fn AssignTeams
  (bind _returnvalue (Minigame|GetMinigameTeamComponent))
  (bind _returnvalue_2 (Minigame|GetMinigameTeamComponent))
  (bind _teamsizes (LocalVariables|GetTeamSizes))
  (bind _fnevoicechatcomponent (Variables|Minigame|GetFNEVoiceChatComponent))
  (bind _currentteam (LocalVariables|GetCurrentTeam))
  (Networking|SwitchHasAuthority
    (:Authority
      (Utilities|IsValid _returnvalue
        (:Is%20Valid
          (bind _output_get (LocalVariables|SetLocalMinigameTeamComponent _returnvalue))
          (LocalVariables|SetTeamSizes (Utilities|Array|MakeArray 0 0 0 0 0 0 0 0 0 0 0 0 0 0 0 0))
          (if (>= (Minigame|Teams|GetMinigameNumTeams _output_get) 0)
            (Utilities|FlowControl|Sequence
              (:then_0
                (LocalVariables|SetCurrentTeam 0)
                (LocalVariables|SetRandomlyOrderedPlayers (CallFunction|GetRandomlyOrderedParticipatingPlayers))
                ;[1 1 1] Cache the max initial player counts for each team
                (Utilities|Array|Resize _teamsizes (Minigame|Teams|GetMinigameNumTeams (LocalVariables|GetLocalMinigameTeamComponent)))
                (for (_ _array_index) _teamsizes
                  (Utilities|IsValid _returnvalue_2
                    (:Is%20Valid
                      (Utilities|Array|Add (LocalVariables|GetMaxInitTeamSizes) (Minigame|GameRules|GetMaxInitTeamSize _returnvalue_2 (CallFunction|IndextoTeamIndex _array_index)))))))
              (:then_1
                (for _array_element (LocalVariables|GetRandomlyOrderedPlayers)
                  (if (not (LocalVariables|GetbAllTeamsAtCapacity))
                    (CallFunction|SetTeamorRefreshVoiceChannel _array_element (CallFunction|IndextoTeamIndex _currentteam))
                    (Math|Integer|IncrementInt (Utilities|Array|Get(aref) (LocalVariables|GetTeamSizes) _currentteam)
                      (:%20%20
                        (LocalVariables|SetbAllTeamsAtCapacity true)
                        (Utilities|FlowControl|ForLoopwithBreak 0 (Utilities|Array|LastIndex (LocalVariables|GetMaxInitTeamSizes))
                          (:LoopBody
                            (bind _output_get_2 (LocalVariables|SetCurrentTeam (Math|Integer|%25%28Integer%29 (+ (LocalVariables|GetCurrentTeam) 1) (Minigame|Teams|GetMinigameNumTeams (LocalVariables|GetLocalMinigameTeamComponent)))))
                            (if (or (< (Utilities|Array|Get(acopy) (LocalVariables|GetMaxInitTeamSizes) _output_get_2) 0) (> (Utilities|Array|Get(acopy) (LocalVariables|GetMaxInitTeamSizes) _output_get_2) (Utilities|Array|Get(acopy) (LocalVariables|GetTeamSizes) _output_get_2)))
                              (LocalVariables|SetbAllTeamsAtCapacity false))))))
                    (else
                      (CallFunction|SetTeamorRefreshVoiceChannel _array_element (Utilities|Enum|LiteralenumEFortTeam "Spectator")))))))
            (else
              (Utilities|IsValid _fnevoicechatcomponent
                (:Is%20Valid
                  (Minigame|RefreshVoiceforPlayers _fnevoicechatcomponent)
                  (for _array_element_2 (Minigame|GetParticipatingPlayers)
                    (Math|Integer|IncrementInt (Utilities|Array|Get(aref) (LocalVariables|GetTeamSizes) (Math|Conversions|ToInteger(Byte) (Math|Byte|Byte-Byte (Team|GetActorTeam _array_element_2) (Utilities|Enum|LiteralenumEFortTeam "HumanPvP_Team1"))))))))))
          (Minigame|Teams|SetTeamMinigame (LocalVariables|GetLocalMinigameTeamComponent) false)
          (bind _array_element_3 (Utilities|Array|ForEachLoopwithBreak (LocalVariables|GetTeamSizes))
            (:LoopBody
              (if (> _array_element_3 1)
                (Minigame|Teams|SetTeamMinigame (LocalVariables|GetLocalMinigameTeamComponent) true)))))))))

(fn GetRandomlyOrderedParticipatingPlayers
  (bind _output_get (LocalVariables|SetParticipatingPlayersCopy (Minigame|GetParticipatingPlayers)))
  (while (> (Utilities|Array|Length _output_get) 0)
    (bind _output_get_2 (LocalVariables|SetRandomPlayerIndex (Math|Random|RandomIntegerinRange 0 (- (Utilities|Array|Length _output_get) 1))))
    (Utilities|Array|Add (LocalVariables|GetRandomizedParticipatingPlayers) (Utilities|Array|Get(acopy) (LocalVariables|GetParticipatingPlayersCopy) _output_get_2))
    (Utilities|Array|RemoveIndex (LocalVariables|GetParticipatingPlayersCopy) (LocalVariables|GetRandomPlayerIndex)))
  (return (LocalVariables|GetRandomizedParticipatingPlayers)))

(fn HasPlayerStartPads
  (params (InPlayerState :type Fort%20Player%20State%20Object%20Reference))
  (bind _returnvalue (Minigame|GetPlayerSpawnComponent))
  (bind _returnvalue_2 (Minigame|GetPlayerSpawnComponent))
  (if (Minigame|IsPlayerinMinigame InPlayerState)
    (Utilities|IsValid _returnvalue
      (:Is%20Valid
        (for _array_element (Variables|/Script/FortniteGame.FortMinigamePlayerSpawnComponent|GetPlayerStartComponents _returnvalue)
          (Utilities|IsValid _array_element
            (:Is%20Valid
              (if (not (Variables|/Script/FortniteGame.FortMinigamePlayerStartComponent|GetbIsCheckpoint _array_element))
                (for _array_element_2 (Variables|/Script/FortniteGame.FortMinigamePlayerStartComponent|GetPlayerStarts _array_element)
                  (Utilities|IsValid _array_element_2
                    (:Is%20Valid
                      (bind _ascontroller (Utilities|Casting|CastToController (Actor|GetOwner (Variables|GetInPlayerState))))
                      (Utilities|IsValid _returnvalue_2
                        (:Is%20Valid
                          (bind _returnvalue_biscompatible (Utilities|Struct|BreakPlayerStartCompatibilityInfo (PlayerSpawn|IsPlayerStartCompatiblewithPlayer _returnvalue_2 _ascontroller _array_element_2)))
                          (if _returnvalue_biscompatible
                            (return false)))))))))))
        (return false)))))

(fn IndexToTeamIndex
  (params (Index :type Integer))
  (return (Math|Byte|Byte+Byte (Math|Conversions|ToByte(Integer) Index) (Utilities|Enum|LiteralenumEFortTeam "HumanPvP_Team1"))))

(fn IsAnyLocalPlayerInMinigame
  (bind _array_element (Utilities|Array|ForEachLoopwithBreak (Fortnite|GetLocalFortPlayerControllers true true))
    (:LoopBody
      (bind _asfort_player_state (Utilities|Casting|CastToFortPlayerState (Variables|/Script/Engine.Controller|GetPlayerState _array_element)))
      (if (or (Minigame|IsPlayerinMinigame _asfort_player_state true) (.bSuccess (Utilities|Casting|CastToFortPlayerControllerSpectating _array_element)))
        (return true)))
    (:Completed
      (return false))))

(fn IsPropHuntGame
  ;[1 1 1] To determine if we're in a prop hunt game we track how many players currently have prop possessors or if they had prop possessors at the start of the game. This is currenly done by an ability  on the prop possessors that increments the minigame's "Number of Hiders in progress". "Number of hiders at game start" is updated at the start of the game if there are more than 0 hiders. Otherwise it is -1. This mimics the previous logic that iterated over player inventories.
  (return (or (> (Variables|Default|GetNumberOfHidersGameStart) 0) (> (Variables|Default|GetNumberOfHidersInProgress) 0))))

(fn OnRep_NumberOfHiders
  (Default|CallOnStartingNumberOfHidersChanged (Variables|Default|GetNumberOfHidersGameStart)))

(fn OnRep_NumberOfHidersInProgress
  (Default|CallOnCurrentNumberOfHidersChanged (Variables|Default|GetNumberOfHidersInProgress)))

(fn OnRep_PropHuntSettings
  (Default|CallOnPropHuntSettingsReady self))

(fn RemovePropHuntHider
  (Networking|CanExecuteCosmeticEvents
    (:False
      (Variables|Default|SetNumberOfHidersInProgress (- (Variables|Default|GetNumberOfHidersInProgress) 1)))))

(fn SetTeamOrRefreshVoiceChannel
  (params (PlayerState :type Fort%20Player%20State%20Object%20Reference) (TeamIndex :type Byte))
  (bind _returnvalue (Minigame|GetMinigameTeamComponent))
  (bind _fnevoicechatcomponent (Variables|Minigame|GetFNEVoiceChatComponent))
  (Utilities|IsValid _returnvalue
    (:Is%20Valid
      (bind _returnvalue_2 (Minigame|SetTeamforPlayer _returnvalue PlayerState TeamIndex true false false "NeverClear" "(TagName=\"\")"))
      (if (not _returnvalue_2)
        (Utilities|IsValid _fnevoicechatcomponent
          (:Is%20Valid
            (Minigame|RefreshVoiceforPlayer _fnevoicechatcomponent PlayerState)))))))

(fn ShouldDropOrDestroyItem
  (params (bShouldDrop :type Boolean%20%28by%20ref%29 :default "false") (bShouldDestroy :type Boolean%20%28by%20ref%29 :default "false") (PC :type Controller%20Object%20Reference) (ItemDef :type Fort%20World%20Item%20Definition%20Object%20Reference))
  (bind _itemdropoverride (Variables|Default|GetItemDropOverride))
  (bind (_value _returnvalue) (Utilities|Map|Find _itemdropoverride ItemDef))
  (if _returnvalue
    (Variables|SetBy-RefVar (Variables|GetbShouldDrop) (Utilities|Select :Index _value :NewEnumerator0 true :NewEnumerator1 false :NewEnumerator2 false))
    (Variables|SetBy-RefVar (Variables|GetbShouldDestroy) (Utilities|Select :Index _value :NewEnumerator0 false :NewEnumerator1 false :NewEnumerator2 true))
    (return false)
    (else
      (return false))))

(fn ShouldShowScoreboard
  (params (PlayerController :type Player%20Controller%20Object%20Reference))
  (bind (_asfort_player_controller_spectating _bsuccess) (Utilities|Casting|CastToFortPlayerControllerSpectating PlayerController))
  (Utilities|FlowControl|Sequence
    (:then_0
      (bind _asfort_player_controller_athena (Utilities|Casting|CastToFortPlayerControllerAthena PlayerController))
      (bind _asfort_player_state (Utilities|Casting|CastToFortPlayerState (Variables|/Script/Engine.Controller|GetPlayerState _asfort_player_controller_athena)))
      (switch Utilities|FlowControl|Switch|SwitchonEFortMinigameState (Variables|Minigame|GetCurrentState)
        (:PostRoundEnd
          _)
        (:PostGameEnd
          _))
      (LocalVariables|SetbShouldShow (Minigame|IsPlayerinMinigame _asfort_player_state true)))
    (:then_1
      (if (not (LocalVariables|GetbShouldShow))
        (LocalVariables|SetbShouldShow _bsuccess)))
    (:then_2
      (return (LocalVariables|GetbShouldShow)))))

(fn ShowPostGame
  (if (Utilities|Enum|NotEqual(Enum) (Utilities|Struct|BreakMinigamePostGameSettings (Minigame|GetPostGameSettings (Minigame|GetMinigameSettingsMachine self))) "Classic")
    (Class|AthenaHUDContext|StartPostGameMenu (BlueprintContext|GetAthenaHUDContext))
    (else
      (CallFunction|ShowScoreboardPanel false))))

(fn UpdatePropHuntSettings
  (Default|CallOnPropHuntSettingsChanged self))

(fn ConstructionScript)
